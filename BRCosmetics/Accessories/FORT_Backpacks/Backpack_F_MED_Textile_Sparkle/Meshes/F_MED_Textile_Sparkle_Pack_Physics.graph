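# Physics asset for the Textile Sparkle backpack cosmetic, providing bodies and joints for physics motion on its handle and zipper pulls.
# A large capsule covers the pack bone, with small capsules on two handle bones and two zipper bones.
# The handle forms a two-link chain from the pack, while each zipper attaches directly to it.
# Tight angular limits keep each joint swinging mainly about one axis, with minimal twist.
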PhysicsAsset /BRCosmetics/Accessories/FORT_Backpacks/Backpack_F_MED_Textile_Sparkle/Meshes/F_MED_Textile_Sparkle_Pack_Physics
bodies {
 dyn_handle_01 [{shapeType=Capsule, length=3.75, radius=1.5, shapeName=dyn_handle_01_capsule, center=<-0.9346 0 0.0889>, rotation=<rot 0 0 90>}]
 dyn_handle_02 [{shapeType=Capsule, length=2.25, radius=1.5, shapeName=dyn_handle_02_capsule, center=<0.6069 0 -0.2046>, rotation=<rot 0 0 90>}]
 dyn_zip_01 [{shapeType=Capsule, length=2.005, radius=0.505, shapeName=dyn_zip_01_capsule, center=<1.2781 0 0>, rotation=<rot 90 0 0>}]
 dyn_zip_02 [{shapeType=Capsule, length=1.9175, radius=0.5174, shapeName=dyn_zip_02_capsule, center=<1.1686 0 0>, rotation=<rot 90 0 0>}]
 pack [{
   shapeType=Capsule
   length=13.647
   radius=9.4706
   shapeName=pack_capsule
   center=<0.1914 -5.7538 -7.0466>
   rotation=<rot 4.6774 -89.721 0.9412>
  }]
}
constraints [
 {bone1Name=dyn_handle_02, bone2Name=dyn_handle_01, swing1LimitDegrees=0, swing2LimitDegrees=8, twistLimitDegrees=1}
 {bone1Name=dyn_handle_01, bone2Name=pack, swing1LimitDegrees=0, swing2LimitDegrees=16, twistLimitDegrees=1}
 {bone1Name=dyn_zip_01, bone2Name=pack, swing1LimitDegrees=25, swing2LimitDegrees=1, twistLimitDegrees=1}
 {bone1Name=dyn_zip_02, bone2Name=pack, swing1LimitDegrees=1, swing2LimitDegrees=10, twistLimitDegrees=1}
]
