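# Animation Blueprint for the F_MED_Textile_Sparkle backpack cosmetic, adding secondary physics motion to the pack.
# The reference pose feeds a RigidBody node at partial alpha, simulating the pack's physics asset in root-bone space with overridden gravity and scaled, clamped component motion.
# An additive Transform Bone then nudges the pack bone slightly along Y before output.
# The event graph holds only unconnected default nodes, with no logic.
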AnimBlueprint /BRCosmetics/Accessories/FORT_Backpacks/Backpack_F_MED_Textile_Sparkle/Meshes/F_MED_Textile_Sparkle_Pack_AnimBP
target_skeleton=&/F_MED_Textile_Sparkle_Pack_Skeleton
graphs [
 AnimGraph {nodes [
   n1 = RigidBody(ComponentPose=Local To Component(LocalPose=Local Space Ref Pose()), Alpha=0.7) {
    bOverrideWorldGravity=true
    overridePhysicsAsset=&/F_MED_Textile_Sparkle_Pack_Physics
    simulationSpace=BaseBoneSpace
    alphaBoolBlend {blendOption=Linear}
    baseBoneRef {boneName=root}
    componentAppliedLinearAccClamp=<3000 3000 3000>
    componentLinearAccScale=<0.2 0.2 0.4>
    componentLinearVelScale=<0.1 0.1 0.2>
    overrideWorldGravity {z=-980}
   }
   n2 = "Transform (Modify) Bone - Bone: pack"(ComponentPose=@n1, Translation="0,1,0", Rotation="0,0,0") {translationMode=BMM_Additive, alphaBoolBlend {blendOption=Linear}, boneToModify {boneName=pack}, translation {y=1}}
   Output Pose(Result=Component To Local(ComponentPose=@n2)) {layerGroup=None}
  ]}
 EventGraph {nodes [
   TryGetPawnOwner()
   Event BlueprintUpdateAnimation()
  ]}
]
variables []
